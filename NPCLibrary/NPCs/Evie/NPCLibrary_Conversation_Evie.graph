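# Conversation database for the NPC Evie, defining the dialogue that opens when a player talks to her.
# The entry point leads to one Speech node with a default greeting and contextual lines for a first-ever conversation or for players wearing specific character cosmetics.
# That node branches through link nodes, by remote entry tag, into shared NPC library quests, basic services and a Season 24 narrative quest conversation for Evie.
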ConversationDatabase /NPCLibrary/NPCs/Evie/NPCLibrary_Conversation_Evie
entry_points [{tag=Conversation.Entry.Evie, destinations [#1]}]
nodes [
 ConversationEntryPointNode_7 {
  id=#1
  class=ConversationEntryPointNode
  outputs [#2]
  properties {entryTag {tagName=Conversation.Entry.Evie}}
 }
 Speech_C_4 {
  id=#2
  class=Speech_C
  outputs [#3, #4, #5]
  properties {
   nodeName=""
   generalConfig {
    defaultMessage=You keepin' the peace?
    contextualMessages [
     {
      message="You hit my hometown yet, Hero?"
      contextRequirements [{requirement {class=FortControllerRequirement_IsFirstEverConversationWithNPC}}]
      priority {curve {rowName=Tandem.Speech.Priority.FirstEverInteraction}}
     }
     {
      message=The Syndicates won't go to war. Not on my watch.
      contextRequirements [{requirement {class=FortControllerRequirement_HasCID, properties {allowedCIds [&Character_LocalZilla]}}}]
      priority {value=0, curve {curveTable=null, rowName=None}}
     }
     {
      message=How's it feel to be back home?
      contextRequirements [{requirement {class=FortControllerRequirement_HasCID, properties {allowedCIds [/BRCosmetics/Athena/Items/Cosmetics/Characters/Character_Candor]}}}]
      priority {value=0, curve {curveTable=null, rowName=None}}
     }
     {
      message=I told you I'd bring you to the city someday.
      contextRequirements [{requirement {
         class=FortControllerRequirement_HasCID
         properties {allowedCIds [/BRCosmetics/Athena/Items/Cosmetics/Characters/CID_A_326_Athena_Commando_M_SharpDresserBlack]}
        }}]
      priority {value=0, curve {curveTable=null, rowName=None}}
     }
     {
      message=Your art could really brighten this Island up.
      contextRequirements [{requirement {class=FortControllerRequirement_HasCID, properties {allowedCIds [/BRCosmetics/Athena/Items/Cosmetics/Characters/Character_Dazzle]}}}]
      priority {value=0, curve {curveTable=null, rowName=None}}
     }
     {
      message="Please don't make me listen to another  evil doomfluencer monologue."
      contextRequirements [{requirement {class=FortControllerRequirement_HasCID, properties {allowedCIds [&Character_Inferno]}}}]
      priority {value=0, curve {curveTable=null, rowName=None}}
     }
    ]
   }
  }
 }
 ConversationLinkNode_2 {
  id=#5
  class=ConversationLinkNode
  properties {remoteEntryTag {tagName=Conversation.Entry.Season24.Narrative.P01.Q07.Evie}}
 }
 ConversationLinkNode_0 {id=#4, class=ConversationLinkNode, properties {remoteEntryTag {tagName=Conversation.Entry.BasicServices}}}
 ConversationLinkNode_1 {id=#3, class=ConversationLinkNode, properties {remoteEntryTag {tagName=Conversation.Entry.NPCLibrary.Quests}}}
]
